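# Unlit translucent, two-sided particle material for glass shards of a New Year disco ball effect, usable on Niagara and particle sprites.
# Each sprite rotates around one of three randomly picked axes, with angle driven by distance from world origin, so movement produces spin.
# A random 2x2 sub-UV picks the shard; its normal map drives bright glints and opacity, masked by shard shape.
# Particle color gets a random hue shift; mobile skips rotation and shift.
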Material /Game/Effects/Fort_Effects/Materials/NewYear/M_NewYear_DiscoBall_Shard
graph {
 properties {
  bUsedWithNiagaraSprites=true
  bUsedWithParticleSprites=true
  blendMode=BLEND_Translucent
  pixelDepthOffsetMode=PDOM_Legacy
  shadingModel=MSM_Unlit
  translucencyPass=MTP_BeforeDOF
  twoSided=true
  usedShadingModels=Unlit
 }
 nodes [
  n1 = ParticleRandom()
  n2 = VectorParameter() {parameterName=NormalizeRotationAxis_1, defaultValue=<rgba 1 0.3 -0.2 0>}
  n3 = VectorParameter() {parameterName=NormalizeRotationAxis_2, defaultValue=<rgba -0.44 1 0.13 0>}
  n4 = VectorParameter() {parameterName=NormalizeRotationAxis_3, defaultValue=<rgba 0.06 -0.22 1 0>}
  n5 = Distance(A=Constant3Vector(), B=ParticlePositionWS())
  n6 = ScalarParameter() {defaultValue=5000, parameterName=Spin Scale}
  n7 = ParticleRandom()
  n8 = RotateAboutAxis(
   NormalizedRotationAxis=Normalize(VectorInput=MaterialFunctionCall!(Random Picker=@n1, Texture Input, Option 1=@n2.RGB, Option 2=@n3.RGB, Option 3=@n4.RGB) {materialFunction=/Game/Effects/Fort_Effects/Materials/_Functions/MF_RGB_Random_01}.Vec3 Out Only)
   RotationAngle=Multiply(A=Divide(A=@n5, B=@n6), B=Clamp(None=@n7) {minDefault=0.55})
   PivotPoint=ParticlePositionWS()
   Position=WorldPosition().XYZ
  )
  n9 = Constant()
  FeatureLevelSwitch(Default=@n8, ES2=@n9, ES3_1=@n9, SM4_REMOVED=@n8)
  n10 = ParticleColor()
  n11 = MaterialFunctionCall(Hue Shift Percentage=ParticleRandom(), Texture=@n10.RGB) {materialFunction=&HueShift}
  n12 = FeatureLevelSwitch(Default=@n11.Result, ES2=@n10.RGB, ES3_1=@n10.RGB, SM4_REMOVED=@n11.Result)
  n13 = Constant() {r=2}
  n14 = MaterialFunctionCall!(UseTexCoord 1, Rows=@n13, Columns=@n13, Particle Random Value) {materialFunction=/Game/Effects/Fort_Effects/Materials/_Functions/MF_RandomSubUV_GPU_Sprites}
  n15 = OneMinus(None=DotProduct(
    A=Constant3Vector() {constant=<rgba 2 0 1 0>}
    B=TextureSample(UVs=@n14.Result) {samplerType=SAMPLERTYPE_Normal, texture=/Game/Effects/Fort_Effects/Textures/Glass/T_Cine_GlassShard}.RGB
   ))
  MaterialFunctionCall(Emissive=LinearInterpolate(A=@n12, B=Multiply(A=@n12) {constB=100}, Alpha=@n15) {constA=1.2, constB=100}) {materialFunction=&MF_FX_EmissiveAdjustments}
  Multiply(
   A=Multiply(
    A=LinearInterpolate(Alpha=@n15) {constA=0.15}
    B=TextureSample(UVs=@n14.Result) {texture=/Game/Effects/Fort_Effects/Textures/Glass/T_Cine_GlassShard_Mask}.G
   )
   B=ParticleColor().A
  )
 ]
}
