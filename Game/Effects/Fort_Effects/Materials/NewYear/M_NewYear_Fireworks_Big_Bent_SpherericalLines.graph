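# Translucent unlit, two-sided mesh-particle material for big New Year firework line bursts.
# SplineThicken sets line width from parameters, particle dynamic LineWidth and vertex-color variation.
# Opacity combines a center-peaked UV profile, a PanSmoke-driven sphere mask sliding along V, two multiplied panning noise samples offset by vertex color, and particle alpha.
# Color is particle color hue-shifted by vertex blue, brightened by the UV profile and V gradient.
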Material /Game/Effects/Fort_Effects/Materials/NewYear/M_NewYear_Fireworks_Big_Bent_SpherericalLines
graph {
 properties {
  bEnableResponsiveAA=true
  bUsedWithMeshParticles=true
  blendMode=BLEND_Translucent
  opacityMaskClipValue=0.1
  pixelDepthOffsetMode=PDOM_Legacy
  shadingModel=MSM_Unlit
  twoSided=true
  usedShadingModels=Unlit
 }
 nodes [
  n1 = VertexColor()
  n2 = ScalarParameter() {parameterName=WidthTip}
  n3 = DynamicParameter() {paramNames [PanGlowTip,PanSmoke,TipSize,LineWidth]}
  n4 = VertexColor()
  MaterialFunctionCall(
   WidthBase=Add(A=ScalarParameter() {defaultValue=500, parameterName=WidthBase}, B=Multiply(B=@n1.G) {constA=25})
   WidthTip=Add(A=Multiply(A=@n2, B=@n3.LineWidth), B=Multiply(B=@n4.G) {constA=2})
  ) {materialFunction=&SplineThicken}
  n5 = TextureCoordinate()
  n6 = ComponentMask(None=@n5) {r=true}
  n7 = Multiply(B=Multiply(A=@n6, B=OneMinus(None=@n6))) {constA=4}
  n8 = Multiply(B=Multiply(A=ComponentMask(None=@n5) {g=true}, B=Power(Base=@n7) {constExponent=4})) {constA=15}
  n9 = Clamp(None=@n8)
  n10 = TextureCoordinate() {uTiling=0}
  n11 = ComponentMask(None=AppendVector(A=Constant(), B=DynamicParameter() {paramNames [PanGlowTip,PanSmoke,TipSize,LineWidth]}.PanSmoke)) {g=true, r=true}
  n12 = ScalarParameter() {defaultValue=0.25, parameterName=GlowTip_Radius_1}
  n13 = ScalarParameter() {parameterName=GlowTip_Hardness_1}
  n14 = VertexColor()
  n15 = Add(A=TextureCoordinate() {uTiling=0.01, vTiling=5}, B=@n14.G)
  n16 = Add(A=@n14.R, B=TextureCoordinate() {uTiling=0.01, vTiling=5})
  n17 = Multiply(
   A=TextureSample(UVs=Panner(Coordinate=@n15) {speedY=0.25}) {texture=/Game/Global/Textures/Default/Noise/T_TileNoise_02}.G
   B=TextureSample(UVs=Panner(Coordinate=@n16) {speedY=0.3}) {texture=&T_TileNoise_02}.G
  )
  n18 = ScalarParameter() {defaultValue=1.25, parameterName=Exp}
  n19 = Multiply(A=Clamp(None=SphereMask(A=@n10, B=@n11, Radius=@n12, Hardness=@n13)), B=Clamp(None=Power(Base=@n17, Exp=@n18) {constExponent=1.25}))
  Multiply(A=Multiply(A=@n9, B=Clamp(None=@n19)), B=ParticleColor().A)
  n20 = ScalarParameter() {defaultValue=-0.25, parameterName=Color_Hueshift_low}
  n21 = ScalarParameter() {defaultValue=0.25, parameterName=Color_Hueshift_High}
  n22 = VertexColor()
  Multiply(
   A=Multiply(A=@n8, B=@n9)
   B=MaterialFunctionCall(Hue Shift Percentage=LinearInterpolate(A=@n20, B=@n21, Alpha=@n22.B), Texture=ParticleColor().RGB) {materialFunction=&HueShift}.Result
  )
 ]
}
