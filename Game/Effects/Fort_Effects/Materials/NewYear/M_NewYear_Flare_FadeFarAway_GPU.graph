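# Unlit additive GPU sprite particle material for a New Year flare, drawing a spiked glow.
# Two sphere-mask falloffs plus a spikeball texture form the shape.
# Its color blends two spark tints, gains HDR intensity from a dynamic parameter, hue-shifts per particle random and scales by particle green before emissive adjustment.
# A separate particle-alpha term uses a powered camera-distance fade, named for dimming flares far away.
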Material /Game/Effects/Fort_Effects/Materials/NewYear/M_NewYear_Flare_FadeFarAway_GPU
graph {
 properties {bUsedWithParticleSprites=true, blendMode=BLEND_Additive, pixelDepthOffsetMode=PDOM_Legacy, shadingModel=MSM_Unlit, usedShadingModels=Unlit}
 nodes [
  n1 = ParticleColor()
  n2 = LinearInterpolate(
   A=VectorParameter() {parameterName=SparkColor0, defaultValue=<rgba 1 0.0978 0 0>}.RGB
   B=VectorParameter() {parameterName=SparkColor1, defaultValue=<rgba 1 0.2289 0 0>}.RGB
   Alpha=ScalarParameter() {parameterName=ColorBlend}
  )
  n3 = DynamicParameter() {defaultValue=<rgba 1 1 -1 1>, paramNames [ColorIntensity, PanSmoke, OffsetMesh, LineWidth]}
  n4 = MaterialFunctionCall(
   Hue Shift Percentage=ParticleRandom()
   Texture=MaterialFunctionCall(LDR Input=@n2, HDR Tint=Multiply(A=@n3.ColorIntensity, B=@n2)) {materialFunction=&DeriveHDRfromLDR}.RemappedRange
  ) {materialFunction=&HueShift}
  n5 = Constant() {r=0.5}
  n6 = SphereMask(A=TextureCoordinate(), B=@n5, Radius=@n5) {hardnessPercent=0}
  n7 = Multiply(A=Multiply(A=Power(Base=@n6) {constExponent=25}) {constB=2}, B=LinearInterpolate(B=Constant() {r=1}) {constA=1, constAlpha=0.7})
  n8 = Add(
   A=Multiply(A=Power(Base=@n6) {constExponent=4}) {constB=0.15}
   B=TextureSampleParameter2D() {parameterName=FlareMask, texture=/Game/Effects/Fort_Effects/Textures/Flares/T_Flare_Spikeball}.RGB
  )
  MaterialFunctionCall(Emissive=Multiply(A=Multiply(A=@n1.G, B=@n4.Result), B=Add(A=@n7, B=@n8))) {materialFunction=&MF_FX_EmissiveAdjustments}
  Multiply(A=ParticleColor().R, B=Clamp(None=Constant() {r=1}) {maxDefault=0.75, minDefault=0.25})
  n9 = Power(Base=MaterialFunctionCall(Distance=ScalarParameter() {defaultValue=25000, parameterName=NearFade}) {materialFunction=&Near_Camera_Fade}.Result) {constExponent=8}
  Multiply(A=Multiply(A=Constant() {r=1}, B=ParticleColor().A), B=LinearInterpolate(A=Constant() {r=1}, B=Constant() {r=0.25}, Alpha=Clamp(None=@n9)))
 ]
}
