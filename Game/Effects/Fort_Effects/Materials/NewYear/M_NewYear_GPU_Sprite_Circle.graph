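# Masked unlit particle sprite material for New Year effects, drawing a colored circular spark for Cascade or Niagara, chosen by static switch.
# Particle color or blended spark colors feed HDR derivation, then a random or dynamic-parameter hue shift and emissive adjustment.
# Mobile feature levels substitute a flat color and fixed hue pick.
# Radially mapped noise, eroded by particle alpha, is cut by a sphere mask, with optional near-camera fade.
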Material /Game/Effects/Fort_Effects/Materials/NewYear/M_NewYear_GPU_Sprite_Circle
graph {
 properties {
  bUsedWithNiagaraSprites=true
  bUsedWithParticleSprites=true
  blendMode=BLEND_Masked
  opacityMaskClipValue=0.1
  pixelDepthOffsetMode=PDOM_Legacy
  shadingModel=MSM_Unlit
  usedShadingModels=Unlit
 }
 nodes [
  n1 = Add(A=TextureCoordinate(), B=Constant2Vector() {g=-0.5, r=-0.5})
  Add(A=TextureCoordinate() {uTiling=5, vTiling=5}, B=@n1)
  n2 = Constant() {r=1}
  n3 = ParticleColor()
  n4 = StaticSwitchParameter(
   True=DynamicParameter() {defaultValue=<rgba 1 1 -1 1>, paramNames [ColorIntensity, HueShiftAmount, HDRIntensity, LineWidth]}.HueShiftAmount
   False=ParticleRandom()
  ) {parameterName=UseInNiagara?}
  n5 = ScalarParameter() {parameterName=Mobile_PickColor}
  n6 = FeatureLevelSwitch(Default=@n4, ES2=@n5, ES3_1=@n5, SM4_REMOVED=@n4, SM5=@n4, SM6=@n4)
  n7 = VectorParameter() {parameterName=SparkColor0, defaultValue=<rgba 1 0.0978 0 0>}
  n8 = StaticSwitchParameter(
   True=ParticleColor().RGB
   False=LinearInterpolate(
    A=@n7.RGB
    B=VectorParameter() {parameterName=SparkColor1, defaultValue=<rgba 1 0.2289 0 0>}.RGB
    Alpha=ScalarParameter() {parameterName=ColorBlend}
   )
  ) {parameterName=UseInNiagara?}
  n9 = DynamicParameter() {defaultValue=<rgba 1 1 -1 1>, paramNames [ColorIntensity, HueShiftAmount, HDRIntensity, LineWidth]}
  n10 = MaterialFunctionCall(
   LDR Input=@n8
   HDR Tint=Multiply(A=@n9.ColorIntensity, B=@n8)
   HDR Intensity Multiplier=StaticSwitchParameter(True=@n9.HDRIntensity, False=Constant() {r=1}) {parameterName=UseInNiagara?}
  ) {materialFunction=&DeriveHDRfromLDR}
  n11 = FeatureLevelSwitch(Default=@n10.RemappedRange, ES2=@n7.RGB, ES3_1=@n7.RGB, SM4_REMOVED=@n10.RemappedRange)
  MaterialFunctionCall(Emissive=Multiply(
    A=StaticSwitchParameter(True=@n2, False=@n3.G) {parameterName=UseInNiagara?}
    B=MaterialFunctionCall(Hue Shift Percentage=@n6, Texture=@n11) {materialFunction=&HueShift}.Result
   )) {materialFunction=&MF_FX_EmissiveAdjustments}
  n12 = MaterialFunctionCall(Vector or UVs=@n1) {materialFunction=&VectorToRadialValue}
  Panner(Coordinate=@n12.Radial Coordinates) {speedY=-1}
  ParticleRandom()
  n13 = ParticleColor()
  n14 = TextureSample(UVs=Add(A=TextureCoordinate() {uTiling=0, vTiling=0}, B=@n12.Radial Coordinates)) {texture=/Game/Global/Textures/Default/Noise/T_TileNoise_02}
  n15 = Multiply(
   A=Multiply(
    A=@n13.A
    B=MaterialFunctionCall!(Dynamic Param=@n13.A, Alpha=@n14.G) {materialFunction=/Game/Effects/Fort_Effects/Materials/_Functions/Erode_Alpha}.Result
   )
   B=SphereMask(
    A=TextureCoordinate()
    B=Constant2Vector() {g=0.5, r=0.5}
    Radius=ScalarParameter() {defaultValue=0.5, parameterName=Radius}
    Hardness=ScalarParameter() {parameterName=Hardness}
   )
  )
  n16 = ScalarParameter() {defaultValue=150, parameterName=Fade Distance}
  StaticSwitchParameter(
   True=Multiply(A=@n15, B=MaterialFunctionCall(Distance=@n16) {materialFunction=&Near_Camera_Fade}.Result)
   False=@n15
  ) {parameterName=Use Near Camera}
 ]
}
