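# Cheap additive, unlit, two-sided light beam material for New Year effects on mesh particles, shaped from UV math with no texture samples.
# World position offset widens the beam with SplineThicken and adds a pivot-relative scale offset, both growing with camera distance.
# Color blends two spark tints through an HDR remap, with particle red driving hue shift and green scaling intensity.
# A UV falloff mask, particle alpha and two near-camera fades control visibility.
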Material /Game/Effects/Fort_Effects/Materials/NewYear/M_NewYear_LightBeam_Cheap
graph {
 properties {
  bUsedWithMeshParticles=true
  blendMode=BLEND_Additive
  pixelDepthOffsetMode=PDOM_Legacy
  shadingModel=MSM_Unlit
  twoSided=true
  usedShadingModels=Unlit
 }
 nodes [
  n1 = DynamicParameter() {defaultValue=<rgba 100 1 1 1>, paramNames [SplineWidth,HueShift,Desat,Param4]}
  n2 = ObjectPositionWS()
  n3 = Distance(A=@n2, B=CameraPositionWS())
  n4 = ScalarParameter() {defaultValue=50, parameterName=DistanceOffset}
  n5 = Clamp(None=Divide(A=Subtract(A=@n3, B=@n4), B=ScalarParameter() {defaultValue=30000, parameterName=ScaleUpDistance}))
  n6 = LinearInterpolate(B=ScalarParameter() {defaultValue=250, parameterName=MaxWidthMult}, Alpha=@n5) {constA=1, constB=3}
  n7 = LinearInterpolate(B=ScalarParameter() {defaultValue=5, parameterName=MaxWidthMult_Tip}, Alpha=@n5) {constA=1, constB=3}
  n8 = Multiply(A=ScalarParameter() {defaultValue=60, parameterName=MaxScaleIncrease}) {constB=-1}
  n9 = Subtract(A=@n2, B=WorldPosition().XYZ)
  Add(
   A=MaterialFunctionCall(WidthBase=Multiply(A=@n1.SplineWidth, B=@n6), WidthTip=Multiply(A=@n1.SplineWidth, B=@n7)) {materialFunction=&SplineThicken}.WorldPositionOffset
   B=LinearInterpolate(B=Multiply(A=@n8, B=@n9) {constB=-1}, Alpha=@n5) {constB=0}
  )
  n10 = ParticleColor()
  n11 = Clamp(None=Constant() {r=1}) {maxDefault=0.75, minDefault=0.25}
  n12 = LinearInterpolate(
   A=VectorParameter() {parameterName=SparkColor0, defaultValue=<rgba 1 0.0978 0 0>}.RGB
   B=VectorParameter() {parameterName=SparkColor1, defaultValue=<rgba 1 0.2289 0 0>}.RGB
   Alpha=ScalarParameter() {parameterName=ColorBlend}
  )
  n13 = Multiply(B=@n12)
  Multiply(
   A=ParticleColor().G
   B=MaterialFunctionCall(
    Hue Shift Percentage=Multiply(A=@n10.R, B=@n11)
    Texture=MaterialFunctionCall(LDR Input=@n12, HDR Tint=@n13) {materialFunction=&DeriveHDRfromLDR}.RemappedRange
   ) {materialFunction=&HueShift}.Result
  )
  n14 = ParticleColor()
  n15 = ComponentMask(None=TextureCoordinate()) {r=true}
  n16 = OneMinus(None=@n15)
  n17 = ComponentMask(None=TextureCoordinate()) {g=true}
  n18 = Power(Base=OneMinus(None=@n17), Exp=ScalarParameter() {defaultValue=6.362, parameterName=PowerFalloff}) {constExponent=4}
  n19 = Clamp(None=LinearInterpolate(A=ScalarParameter() {parameterName=ContrastCheap1}, B=ScalarParameter() {defaultValue=8.1353, parameterName=ContrastCheap2}, Alpha=@n17))
  n20 = Multiply(A=Multiply(A=Multiply(A=@n16, B=@n15), B=Multiply(A=@n18, B=@n19)), B=ScalarParameter() {defaultValue=0.1, parameterName=Brightness})
  n21 = MaterialFunctionCall(Distance=ScalarParameter() {defaultValue=15000, parameterName=NearCamFade_Distance}) {materialFunction=&Near_Camera_Fade}
  n22 = MaterialFunctionCall(Distance=ScalarParameter() {defaultValue=5000, parameterName=TurnOff_Beam_Distance}) {materialFunction=&Near_Camera_Fade}
  Multiply(A=Multiply(A=Multiply(A=@n14.A, B=@n20)), B=Clamp(None=Multiply(A=@n21.Result, B=@n22.Result)))
 ]
}
