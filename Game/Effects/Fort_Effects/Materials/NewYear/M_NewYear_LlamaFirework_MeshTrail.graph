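# Masked, two-sided lit material for the mesh trail of the New Year llama firework, rendered as billowing, glowing smoke.
# Panning storm-cloud textures tiled by world height drive color, glow, normals and normal-direction vertex displacement, with dynamic parameters for tiling, thickness, bottom pinch and glow height.
# The MPC_GeodeNick ZHeight value controls the trail's hot lip and vertex collapse.
# Particle alpha erodes the view-dependent mask; mobile swaps in cheaper sine displacement and simpler glow.
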Material /Game/Effects/Fort_Effects/Materials/NewYear/M_NewYear_LlamaFirework_MeshTrail
graph {
 properties {
  bUsedWithMeshParticles=true
  bUsedWithParticleSprites=true
  bUsedWithSplineMeshes=true
  blendMode=BLEND_Masked
  pixelDepthOffsetMode=PDOM_Legacy
  translucencyLightingMode=TLM_SurfacePerPixelLighting
  twoSided=true
  usedShadingModels=Default Lit
 }
 nodes [
  n1 = ScalarParameter() {defaultValue=15000, parameterName=HotLipFalloff}
  n2 = WorldPosition()
  n3 = Subtract(
   A=CollectionParameter() {collection=/Game/Effects/Fort_Effects/Materials/Geode/MPC_GeodeNick, parameterName=ZHeight}
   B=MaterialFunctionCall(Float3=@n2.XYZ) {materialFunction=&BreakOutFloat3Components}.B
  )
  n4 = Divide(A=@n3, B=ScalarParameter() {defaultValue=3006.0029, parameterName=VertCollapseFalloff})
  n5 = OneMinus(None=Clamp(None=Divide(A=@n4, B=@n1)))
  n6 = ScalarParameter() {defaultValue=1, group=TextureControl, parameterName=ScaleU}
  n7 = Multiply(
   A=@n6
   B=MaterialFunctionCall(Float2=TextureCoordinate() {uTiling=0.5}) {materialFunction=&BreakOutFloat2Components}.R
  )
  n8 = Divide(
   A=MaterialFunctionCall(Float3=WorldPosition().XYZ) {materialFunction=&BreakOutFloat3Components}.B
   B=ScalarParameter() {defaultValue=-62648.6055, parameterName=ZWorldTiling}
  )
  n9 = Multiply(
   A=DynamicParameter() {defaultValue=<rgba 25 250 1 1>, paramNames [BlobTiling, Thickness, Bottom Close, GlowHeight]}.BlobTiling
   B=ScalarParameter() {defaultValue=0.5, parameterName=TileMultiply}
  )
  n10 = Multiply(A=Multiply(A=ScalarParameter() {defaultValue=1, parameterName=LargestScaleMultiply}, B=@n8), B=@n9)
  n11 = Multiply(A=Time(), B=ScalarParameter() {defaultValue=0.001, group=TextureControl, parameterName=Speed})
  n12 = TextureSample(UVs=Panner(Coordinate=AppendVector(A=@n7, B=@n10), Time=@n11) {bFractionalPart=true, speedX=0.05, speedY=-0.1}) {samplerType=SAMPLERTYPE_LinearGrayscale, texture=/Game/Global/Textures/Default/Tileable_Patterns/T_StormCloudDetail01_M}
  n13 = MaterialFunctionCall(Float2=TextureCoordinate() {vTiling=0.6}) {materialFunction=&BreakOutFloat2Components}
  n14 = Multiply(A=@n8) {constB=0.6}
  n15 = Panner(Coordinate=AppendVector(A=Multiply(A=@n6, B=@n13.R), B=Multiply(A=@n14, B=@n9)), Time=@n11) {bFractionalPart=true, speedX=-0.06, speedY=-1}
  n16 = TextureSample(UVs=@n15) {samplerType=SAMPLERTYPE_LinearGrayscale, texture=&T_StormCloudDetail01_M}
  n17 = Add(A=@n12.G, B=@n16.G)
  n18 = MaterialFunctionCall(Float2=TextureCoordinate() {vTiling=2}) {materialFunction=&BreakOutFloat2Components}
  n19 = Multiply(A=@n8) {constB=0.4}
  n20 = Add(A=AppendVector(A=Multiply(A=@n6, B=@n18.R), B=Multiply(A=@n19, B=@n9))) {constB=0.4}
  n21 = ScalarParameter() {defaultValue=1, parameterName=FastPannerSpeedU}
  n22 = ScalarParameter() {defaultValue=-1, parameterName=FastPannerSpeedV}
  n23 = TextureSample(UVs=Panner(Coordinate=@n20, Time=@n11, Speed=AppendVector(A=@n21, B=@n22)) {bFractionalPart=true, speedX=20, speedY=-80}) {samplerType=SAMPLERTYPE_LinearGrayscale, texture=&T_StormCloudDetail01_M}
  n24 = Add(A=@n12.G, B=@n23.G)
  n25 = Multiply(A=Multiply(A=@n5, B=LinearInterpolate(A=@n17, B=@n24, Alpha=@n5)), B=ParticleColor().RGB)
  n26 = Multiply(A=ScalarParameter() {defaultValue=0.05, group=Glow, parameterName=SubsurfaceBoost}, B=@n25)
  n27 = ScalarParameter() {defaultValue=0.1, group=Glow, parameterName="Desaturate Luminance (S)_1"}
  n28 = Constant()
  n29 = AppendVector(A=@n28, B=@n28)
  n30 = DynamicParameter() {defaultValue=<rgba 25 250 1 1>, paramNames [BlobTiling, Thickness, Bottom Close, GlowHeight]}
  n31 = Multiply(
   A=MaterialFunctionCall!(
    Location
    Offset=AppendVector(A=@n29, B=@n30.GlowHeight)
    Radius=ScalarParameter() {defaultValue=150000, parameterName="Radius (S)"}
    Calculate Camera Inside
    Depth Biased Alpha
    FadeDistance
   ) {materialFunction=/Engine/Functions/Engine_MaterialFunctions01/Gradient/SphereGradient-3D}.Result 0-1
   B=@n25
  )
  n32 = Add(
   A=MaterialFunctionCall(LDR Input=Clamp(None=@n26), HDR Tint=@n26, Desaturate Luminance=@n27) {materialFunction=&DeriveHDRfromLDR}.RemappedRange
   B=MaterialFunctionCall(LDR Input=Clamp(None=@n31), HDR Tint=@n31, Desaturate Luminance=@n27) {materialFunction=&DeriveHDRfromLDR}.RemappedRange
  )
  Add(
   A=FeatureLevelSwitch(Default=@n32, ES2=@n31, ES3_1=@n31, SM4_REMOVED=@n32)
   B=VectorParameter() {parameterName=BaseEmissiveColor, defaultValue=<rgba 0.01 0.0037 0 1>}.RGB
  )
  n33 = Panner(Coordinate=@n20, Time=@n11) {bFractionalPart=true, speedX=0.05, speedY=-1}
  n34 = Multiply(
   A=VectorParameter() {parameterName=Normallntensity, defaultValue=<rgba 0.2 0.2 1 1>}.RGB
   B=MaterialFunctionCall(
    BaseNormal=TextureSample(UVs=@n33) {samplerType=SAMPLERTYPE_Normal, texture=/Game/Global/Textures/Default/Tileable_Patterns/T_StormCloudDetail01_N}.RGB
    AdditionalNormal=TextureSample(UVs=@n15) {samplerType=SAMPLERTYPE_Normal, texture=&T_StormCloudDetail01_N}.RGB
   ) {materialFunction=&BlendAngleCorrectedNormals}.Result
  )
  FeatureLevelSwitch(Default=@n34, ES2=@n34, ES3_1=@n34, SM4_REMOVED=@n34)
  n35 = VertexNormalWS()
  n36 = Divide(A=@n3, B=ScalarParameter() {defaultValue=600000, parameterName=MacroScale})
  n37 = Multiply(A=ScalarParameter() {defaultValue=1000, parameterName=MacroWidthAmount}, B=Clamp(None=@n36))
  n38 = Clamp(None=Multiply(A=@n36, B=ScalarParameter() {defaultValue=1.7783, parameterName=FalloffHardness}))
  n39 = ScalarParameter() {defaultValue=1, group=WPO, parameterName=Valley}
  n40 = LinearInterpolate(A=ScalarParameter() {defaultValue=2.5, parameterName=PeaksMin}, B=ScalarParameter() {defaultValue=5, parameterName=PeaksMax}, Alpha=@n38)
  n41 = LinearInterpolate(A=Multiply(A=@n12.G, B=@n16.G), B=Multiply(A=@n23.G, B=@n16.G), Alpha=@n5)
  n42 = Multiply(
   A=Multiply(A=LinearInterpolate(A=@n39, B=@n40, Alpha=@n41) {constA=-1, constB=2}, B=VertexNormalWS())
   B=DynamicParameter() {defaultValue=<rgba 25 250 1 1>, paramNames [BlobTiling, Thickness, Bottom Close, GlowHeight]}.Thickness
  )
  n43 = ScalarParameter() {defaultValue=1.5, group=WPO, parameterName=ValleyMobile}
  n44 = LinearInterpolate(
   A=ScalarParameter() {defaultValue=2, parameterName=PeaksMinMobile}
   B=ScalarParameter() {defaultValue=2.2, parameterName=PeaksMaxMobile}
   Alpha=@n38
  )
  n45 = Divide(A=@n8, B=ScalarParameter() {defaultValue=0.0644, parameterName=SineTiling})
  n46 = Divide(A=@n8, B=ScalarParameter() {defaultValue=0.1215, parameterName=SineTiling2})
  n47 = Add(
   A=ConstantBiasScale(None=Sine(None=@n45))
   B=Multiply(A=Sine(None=@n46), B=ScalarParameter() {defaultValue=0.33, parameterName=SecondSineMulti})
  )
  n48 = Multiply(
   A=Multiply(A=LinearInterpolate(A=@n43, B=@n44, Alpha=@n47) {constA=-1, constB=2}, B=VertexNormalWS())
   B=ScalarParameter() {defaultValue=200, group=WPO, parameterName=DeformMobile, sortPriority=2}
  )
  n49 = Multiply(
   A=Add(A=@n38, B=ScalarParameter() {parameterName=MinMacroMultiply})
   B=FeatureLevelSwitch(Default=@n42, ES2=@n48, ES3_1=@n48, SM4_REMOVED=@n42)
  ) {constA=1}
  n50 = MaterialFunctionCall() {materialFunction=&LinearGradient}
  n51 = ScalarParameter() {defaultValue=1, group=WPO, parameterName=ExpandPower}
  n52 = Multiply(A=Multiply(A=Power(Base=@n50.VGradient, Exp=@n51), B=VertexNormalWS()), B=ScalarParameter() {group=WPO, parameterName=Expand})
  n53 = Power(Base=OneMinus(None=Clamp(None=@n4)), Exp=ScalarParameter() {defaultValue=2, parameterName=pow})
  n54 = Subtract(
   A=Add(
    A=MaterialFunctionCall() {materialFunction=&ObjectPivotPoint}.Object Pivot Location
    B=VectorParameter() {parameterName=Pinch Offset, defaultValue=<rgba 0 0 5750 0>}.RGB
   )
   B=@n2.XYZ
  )
  n55 = Clamp(None=Divide(A=ComponentMask(None=@n54) {b=true}, B=ScalarParameter() {defaultValue=1250, parameterName=Falloff Bottom}))
  n56 = LinearInterpolate(
   A=@n53
   B=Add(A=Multiply(A=@n55, B=@n55), B=@n53)
   Alpha=DynamicParameter() {defaultValue=<rgba 25 250 1 1>, paramNames [BlobTiling, Thickness, Bottom Close, GlowHeight]}.Bottom Close
  )
  LinearInterpolate(
   A=Add(A=Multiply(A=@n35, B=@n37), B=Add(A=@n49, B=@n52))
   B=Multiply(A=ScalarParameter() {defaultValue=-0.3937, parameterName=NormalCollapseAmount}, B=@n35)
   Alpha=FeatureLevelSwitch(Default=@n56, ES2=@n53, ES3_1=@n53, SM4_REMOVED=@n56)
  )
  n57 = ScalarParameter() {defaultValue=1, parameterName=ColorIntensity}
  n58 = LinearInterpolate(Alpha=ParticleColor().A) {constB=2}
  Multiply(
   A=Clamp(None=Power(Base=@n17, Exp=@n57))
   B=LinearInterpolate(
    A=VectorParameter() {group=BaseColor, parameterName=BaseColor, defaultValue=<rgba 0.6979 0.8112 1 1>}.RGB
    B=VectorParameter() {group=BaseColor, parameterName=BaseColor2, defaultValue=<rgba 1 0.3599 0 1>}.RGB
    Alpha=Clamp(None=@n58)
   )
  )
  n59 = Divide(
   A=Add(A=PixelDepth().R, B=ScalarParameter() {defaultValue=-400, parameterName=NearClipBias})
   B=ScalarParameter() {defaultValue=100, parameterName=NearClipDistance}
  )
  n60 = Constant() {r=1}
  n61 = ParticleColor()
  n62 = TextureObject() {samplerType=SAMPLERTYPE_LinearColor, texture=/Game/Global/Textures/Default/Tileable_Patterns/Cloud_Diffuse}
  n63 = Multiply(A=VectorParameter() {parameterName=TextureSIze, defaultValue=<rgba 5000 5000 6000 1>}.RGB) {constB=3}
  n64 = TextureObject() {samplerType=SAMPLERTYPE_Normal, texture=/Game/Global/Textures/Default/Tileable_Patterns/Cloud_normals}
  n65 = WorldPosition()
  n66 = DotProduct(
   A=LinearInterpolate(
    A=MaterialFunctionCall!(
     Diffuse Texture Object=@n62
     Spec Texture Object
     TextureSize=@n63
     --------------------------
     Use seperate settings for Z projection
     Use seperate texture for Z projection
     Texture Size Z projection
     Z Plane Diffuse Texture Object
     Z Plane Spec Texture Object
     "--------------------------      "
     ProjectionTransitionContrast
     Use Texture Alpha Heightmap to Refine the Transistion
     HeightLerpTransitionContrast
     PickTheDiffuseHeightChannel
     TheDiffuseHeightChannel
     "--------------------------       "
     NormalTextureObject=@n64
     World Space Normals
     "--------------------------  "
     Export Float 4 Diffuse
     Export Float 4 Spec
     "--------------------------     "
     WorldPosition=@n65.XYZ
     World Space Normal
    ) {materialFunction=/Engine/Functions/Engine_MaterialFunctions02/Texturing/WorldAlignedTextures_Complex}.Normal XYZ Texture
    B=VertexNormalWS()
    Alpha=Constant() {desc=.2, r=0.2}
   ) {desc=Soft normal}
   B=CameraVectorWS()
  )
  n67 = DotProduct(A=VertexNormalWS(), B=CameraVectorWS())
  n68 = Divide(A=Add(A=Clamp(None=@n66), B=Clamp(None=@n67)))
  n69 = Constant() {r=1}
  n70 = Constant()
  Multiply(
   A=FeatureLevelSwitch(Default=@n59, ES2=@n60, ES3_1=@n60, SM4_REMOVED=@n59)
   B=Multiply(
    A=MaterialFunctionCall(Transition Phase=@n61.A, Height Texture=@n68) {materialFunction=&HeightLerp}.Alpha
    B=If(A=@n5, B=@n69, "A > B"=@n70, "A == B"=@n70, "A < B"=@n69)
   )
  )
  ScalarParameter() {parameterName=Metallic}
  ScalarParameter() {parameterName=Specular}
  ScalarParameter() {defaultValue=1, parameterName=Roughness}
 ]
}
